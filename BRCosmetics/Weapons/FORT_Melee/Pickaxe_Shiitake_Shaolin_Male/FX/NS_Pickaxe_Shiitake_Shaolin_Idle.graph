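# Idle cosmetic Niagara effect for the Shiitake Shaolin male pickaxe: a pulsing gem glow with twinkling sparkles.
# Gem burst-spawns a looping-lifetime particle drawn by sprite, mesh and light renderers, with curves cycling its blue-cyan-purple color, alpha, size and light intensity.
# Sparkles continuously spawns short-lived flickering sprites that drift on random velocity and vector noise.
# The movement-spawned BitsWorld trail emitter is disabled, as are all skeletal mesh location modules.
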NiagaraSystem /BRCosmetics/Weapons/FORT_Melee/Pickaxe_Shiitake_Shaolin_Male/FX/NS_Pickaxe_Shiitake_Shaolin_Idle
dependencies {
 usedDataInterfaces [&NiagaraDataInterfaceColorCurve, &NiagaraDataInterfaceCurve, &NiagaraDataInterfaceArrayFloat]
 usedDynamicInputs [
  &Multiply_Vector2D_ByFloat
  &MakeVector2D
  &SelectFloatFromArray
 ]
 usedModules [&SystemState]
}
emitters {
 BitsWorld {
  bEnabled=false
  emitterUpdateScript [
   &EmitterState()
   &SpawnRate(SpawnRate=10, Use Spawn Probability=false) {enabled=false}
   &SpawnPerUnit(Module.Use Max Movement Threshold=true, Module.Use Movement Tolerance=false, Spawn Spacing=100, Use Spawn Probability=false)
  ]
  particleSpawnScript [
   &InitializeParticle(
    Lifetime=&RandomRangeFloat(Minimum=0.5, Maximum=0.8)
    Position Offset=<0 -10 -5>
    Sprite Rotation Mode=::NewEnumerator2
    Sprite Size Mode=::NewEnumerator1
    Uniform Sprite Size Max=7
    Uniform Sprite Size Min=2
    UsePositionOffset=false
    Write Material Random=true
   )
   &CylinderLocation(
    Cylinder Axis:Vector3f=<0 0 1>
    Cylinder Height=15
    Cylinder Radius=3
    MaskSpawnGroup=false
    Override Seed=false
    Rotation Mode:ENiagara_RotationMode=::NewEnumerator1
    Surface Only=true
    Use Endcaps In Surface Only Mode=false
    UseOwnerScale=true
   )
   &SkeletalMeshLocation(
    MaskSpawnGroup=false
    Mesh Sampling Type=::NewEnumerator5
    Position Sampling=::NewEnumerator1
    Sampled Position Offset=&RandomRangeVector(Minimum=<-3 -3 -3>, Maximum=<3 3 3>)
    Skeletal Mesh=$User.NewNiagaraDataInterfaceSkeletalMesh
    Socket Sampling Mode=::NewEnumerator5
   ) {enabled=false}
   &AddVelocityFromPoint(Module.Use Falloff Distance=false, Velocity Strength=&RandomRangeFloat(Minimum=100, Maximum=200)) {enabled=false}
   &InheritVelocity(Inherited Velocity Amount Scale=<0.5 0.5 0.5>, Speed Limit=true, UseSpeedThreshold=false)
  ]
  particleUpdateScript [
   &ParticleState()
   &InheritSourceMovement(Applied Position Scale=<0.5 0.5 0.5>, ApplyForce=false, ApplyPosition=true, ApplyPositionFalloff=false)
   &CurlNoiseForce(Noise Strength=100, Pan Noise=false)
   &GravityForce(Gravity=&RandomRangeVector(Minimum=<0 0 -25>, Maximum=<0 0 -50>))
   &Drag(Drag=5, Use Linear Drag=true, Use Rotational Drag=false)
   &SolveForcesAndVelocity(Clamp Velocity=false, Limit Acceleration=false)
   &DynamicMaterialParameters(
    Index 0 Param 1=&FloatFromCurve(FloatCurve=DataInterface() {propertyValues {Curve {keys [<at 0 1>,<at 1 0>]}}})
    Index 0 Param 2=&UniformRangedFloat(Minimum=20, Maximum=24)
    Index 0 Param 3=5
    Index 0 Param 4=&FloatFromCurve(FloatCurve=DataInterface() {propertyValues {Curve {keys [<at 0.0123 0.25>,<at 0.99 -0.3>,<at 1 -0.4>]}}})
    Module.Param0WriteEnabled=true
    Module.Param1WriteEnabled=true
    Module.Param2WriteEnabled=true
    Module.Param3WriteEnabled=true
   )
   &ScaleColor(
    Scale Alpha=&FloatFromCurve(
     FloatCurve=DataInterface() {propertyValues {Curve {keys [
         {interpMode=RCIM_Cubic, tangentMode=RCTM_User, time=0, value=1}
         {arriveTangent=-1, interpMode=RCIM_Cubic, leaveTangent=-1, time=0.5, value=0.8}
         {arriveTangent=-0.778, interpMode=RCIM_Cubic, leaveTangent=-0.778, tangentMode=RCTM_User, time=1, value=0}
        ]}}}
     Scale Curve=&MakeCustomFloatFromBool(Bool=$User.bIsFrontEndPreview, True Float=1.5, False Float=0.5)
    )
    Scale RGB=&MakeVectorFromLinearColorRGB(Color=&ColorFromCurve(ColorCurve=DataInterface() {propertyValues {
        AlphaCurve {keys [<at 0 1 RCIM_Cubic>]}
        BlueCurve {keys [
          {arriveTangent=-2.1621, interpMode=RCIM_Cubic, leaveTangent=-2.1621, time=0.0093, value=3.142}
          {arriveTangent=-2.1621, interpMode=RCIM_Cubic, leaveTangent=-2.1621, time=1, value=1}
         ]}
        GreenCurve {keys [
          {arriveTangent=-3.1874, interpMode=RCIM_Cubic, leaveTangent=-3.1874, time=0.0093, value=4}
          {arriveTangent=-3.1874, interpMode=RCIM_Cubic, leaveTangent=-3.1874, time=1, value=0.8423}
         ]}
        RedCurve {keys [<at 0.0093 0 RCIM_Cubic>,<at 1 0 RCIM_Cubic>]}
       }}))
    ScaleA=true
    ScaleRGB=true
   )
   &ScaleSpriteSize(Uniform Curve Sprite Scale=DataInterface() {propertyValues {Curve {keys [
        {arriveTangent=-1.3333, interpMode=RCIM_Cubic, leaveTangent=-1.3333, time=0, value=1}
        {arriveTangent=-0.4, interpMode=RCIM_Cubic, leaveTangent=-0.4, time=0.3, value=0.6}
        {arriveTangent=-0.5, interpMode=RCIM_Cubic, leaveTangent=-0.5, time=0.5, value=0.8}
        {arriveTangent=-0.5, interpMode=RCIM_Cubic, leaveTangent=-0.5, time=0.7, value=0.4}
        {arriveTangent=-1.3333, interpMode=RCIM_Cubic, leaveTangent=-1.3333, time=0.9, value=0.6}
        {arriveTangent=-6, interpMode=RCIM_Cubic, leaveTangent=-6, time=1, value=0}
       ]}}})
  ]
  rendererClasses [&NiagaraSpriteRendererProperties]
  renderers [{rendererClass=&NiagaraSpriteRendererProperties}]
 }
 Gem {
  emitterUpdateScript [
   &EmitterState(Life Cycle Mode=::NewEnumerator1, Loop Behavior=::NewEnumerator1, Loop Duration Mode=::NewEnumerator1, UseLoopDelay=false)
   &SpawnBurst_Instantaneous(Use Spawn Probability=false, UseLoopCountLimit=false)
  ]
  particleSpawnScript [
   &InitializeParticle(
    Color=<rgba 5 5 5 5>
    Mesh Scale Mode=::NewEnumerator3
    Position Offset=<0 0 -1.5>
    Sprite Size Mode=::NewEnumerator3
    Uniform Sprite Size=125
    UsePositionOffset=true
    Write Material Random=true
   )
   &InitialMeshOrientation(Mesh Orientation Mode=::NewEnumerator1, Rotation=<-0.05 -0.25 0>, Use Rotation Vector=true) {enabled=false}
   &SkeletalMeshLocation(
    MaskSpawnGroup=false
    Mesh Sampling Type=::NewEnumerator5
    Sampled Position Offset=<0 0 -1.5>
    Skeletal Mesh=$User.New Skeletal Mesh
    Socket Sampling Mode=::NewEnumerator5
   ) {enabled=false}
  ]
  particleUpdateScript [
   &ParticleState(
    Kill Particles When Lifetime Has Elapsed=false
    Let Infinitely Lived Particles Die When Emitter Deactivates=true
    Loop Particles Lifetime=true
   )
   &CameraOffset(Camera Offset Amount=0.1)
   &ScaleColor(
    Scale Alpha=&FloatFromCurve(FloatCurve=DataInterface() {propertyValues {Curve {keys [
         {arriveTangent=0.0924, interpMode=RCIM_Cubic, leaveTangent=0.0924, tangentMode=RCTM_User, time=0, value=0}
         {interpMode=RCIM_Cubic, tangentMode=RCTM_User, time=0.5, value=1}
         {arriveTangent=-0.3637, interpMode=RCIM_Cubic, leaveTangent=-0.3637, tangentMode=RCTM_User, time=1, value=0}
        ]}}})
    Scale RGB=&MakeVectorFromLinearColorRGB(Color=&ColorFromCurve(ColorCurve=DataInterface() {propertyValues {
        AlphaCurve {keys [<at 0 1 RCIM_Cubic>]}
        BlueCurve {keys [<at 0 1 RCIM_Cubic>,<at 0.4848 1 RCIM_Cubic>,<at 1 1 RCIM_Cubic>]}
        GreenCurve {keys [
          {arriveTangent=1.0015, interpMode=RCIM_Cubic, leaveTangent=1.0015, tangentMode=RCTM_User, time=0, value=0.4342}
          {arriveTangent=0.1381, interpMode=RCIM_Cubic, leaveTangent=0.1381, time=0.4848, value=0.7624}
          {arriveTangent=-0.7397, interpMode=RCIM_Cubic, leaveTangent=-0.7397, tangentMode=RCTM_User, time=1, value=0.5723}
         ]}
        RedCurve {keys [
          {arriveTangent=-0.9724, interpMode=RCIM_Cubic, leaveTangent=-0.9724, tangentMode=RCTM_User, time=0, value=0.1549}
          {arriveTangent=0.0031, interpMode=RCIM_Cubic, leaveTangent=0.0031, time=0.4848, value=0}
          {arriveTangent=0.7691, interpMode=RCIM_Cubic, leaveTangent=0.7691, tangentMode=RCTM_User, time=1, value=0.158}
         ]}
       }}))
    ScaleA=true
    ScaleRGB=true
   )
   &FN_Sprite_Attributes(
    Color=&ColorFromCurve(ColorCurve=DataInterface() {propertyValues {
       AlphaCurve {keys [<at 0.0006 0.5>,<at 0.5113 1 RCIM_Cubic>,<at 0.9929 0.5>]}
       BlueCurve {keys [<at 0 1 RCIM_Cubic>,<at 0.4974 1>,<at 1 1 RCIM_Cubic>]}
       GreenCurve {keys [
         {arriveTangent=1.3771, interpMode=RCIM_Cubic, leaveTangent=1.3771, time=0, value=0.0774}
         <at 0.4974 0.7624>
         {arriveTangent=-1.3628, interpMode=RCIM_Cubic, leaveTangent=-1.3628, time=1, value=0.0774}
        ]}
       RedCurve {keys [
         {arriveTangent=-1.2835, interpMode=RCIM_Cubic, leaveTangent=-1.2835, time=0, value=0.6384}
         <at 0.4974 0>
         {arriveTangent=1.2702, interpMode=RCIM_Cubic, leaveTangent=1.2702, time=1, value=0.6384}
        ]}
      }})
    Create Color Attribute=true
    Create SpriteSize Attribute=true
    SpriteSize=&Vector2DFromFloat(Value=&FloatFromCurve(
      FloatCurve=DataInterface() {propertyValues {Curve {keys [
          {interpMode=RCIM_Cubic, tangentMode=RCTM_User, time=0, value=0.5}
          {interpMode=RCIM_Cubic, tangentMode=RCTM_User, time=0.5, value=1}
          {tangentMode=RCTM_User, time=1, value=0.5}
         ]}}}
      Scale Curve=30
     ))
    Write to Module Namespace=true
   )
   &SpriteRotationRate(Rotation Rate=30)
   &FN_Light_Attributes(
    Color=$Particles.Color
    Create Color Attribute=true
    Create Exponent Attribute=true
    Create Position Attribute=true
    Create Radius Attribute=true
    Exponent=6
    Radius=75
    Scale Color=&FloatFromCurve(
     FloatCurve=DataInterface() {propertyValues {Curve {keys [
         {interpMode=RCIM_Cubic, tangentMode=RCTM_User, time=0, value=0.5}
         {interpMode=RCIM_Cubic, tangentMode=RCTM_User, time=0.5, value=1}
         {tangentMode=RCTM_User, time=1, value=0.5}
        ]}}}
     Scale Curve=3
    )
   )
  ]
  rendererClasses [
   &NiagaraSpriteRendererProperties
   &NiagaraMeshRendererProperties
   &NiagaraLightRendererProperties
  ]
  renderers [
   {rendererClass=&NiagaraSpriteRendererProperties}
   {rendererClass=&NiagaraMeshRendererProperties, rendererIndex=1}
   {rendererClass=&NiagaraLightRendererProperties, rendererIndex=2}
  ]
 }
 Sparkles {
  emitterUpdateScript [
   &EmitterState()
   &SpawnRate(SpawnRate=8, Use Spawn Probability=false)
  ]
  particleSpawnScript [
   &InitializeParticle(
    Color=<rgba 1 1 1 2>
    Lifetime Max=1.5
    Lifetime Mode=::NewEnumerator1
    Position Offset=&RandomRangeVector(Minimum=<-2 -1 -1>, Maximum=<2 1 1>)
    Sprite Size Mode=::NewEnumerator1
    Uniform Sprite Size Max=15
    UsePositionOffset=true
    Write Material Random=true
   )
   &SkeletalMeshLocation(
    MaskSpawnGroup=false
    Mesh Sampling Type=::NewEnumerator5
    Sampled Position Offset=&RandomRangeVector(Minimum=<-2.5 -1 -1>, Maximum=<2.5 1 1>)
    Skeletal Mesh=$User.New Skeletal Mesh
    Socket Sampling Mode=::NewEnumerator5
   ) {enabled=false}
   &AddVelocity(Velocity=&RandomRangeVector(Minimum=<-25 -25 -25>, Maximum=<25 25 25>))
  ]
  particleUpdateScript [
   &ParticleState()
   &ScaleColor(
    Scale Alpha=&FloatFromCurve(FloatCurve=DataInterface() {propertyValues {Curve {keys [
         {arriveTangent=3.3997, interpMode=RCIM_Cubic, leaveTangent=3.3997, tangentMode=RCTM_User, time=0, value=0}
         {interpMode=RCIM_Cubic, tangentMode=RCTM_User, time=0.5, value=1}
         {arriveTangent=-2.7937, interpMode=RCIM_Cubic, leaveTangent=-2.7937, tangentMode=RCTM_User, time=1, value=0}
        ]}}})
    Scale RGB=&MakeVectorFromLinearColorRGB(Color=&ColorFromCurve(ColorCurve=DataInterface() {propertyValues {
        AlphaCurve {keys [<at 0 1 RCIM_Cubic>]}
        BlueCurve {keys [
          <at 0 1 RCIM_Cubic>
          {arriveTangent=-0.4858, interpMode=RCIM_Cubic, leaveTangent=-0.4858, time=0.4848, value=1}
          {arriveTangent=-0.9429, interpMode=RCIM_Cubic, leaveTangent=-0.9429, time=1, value=0.5142}
         ]}
        GreenCurve {keys [
          {arriveTangent=1.0015, interpMode=RCIM_Cubic, leaveTangent=1.0015, tangentMode=RCTM_User, time=0, value=0.3181}
          {arriveTangent=0.6819, interpMode=RCIM_Cubic, leaveTangent=0.6819, time=0.4848, value=0.9764}
          {arriveTangent=-0.7397, interpMode=RCIM_Cubic, leaveTangent=-0.7397, tangentMode=RCTM_User, time=1, value=1}
         ]}
        RedCurve {keys [
          {arriveTangent=-0.9724, interpMode=RCIM_Cubic, leaveTangent=-0.9724, tangentMode=RCTM_User, time=0, value=0}
          <at 0.4848 0.0365 RCIM_Cubic>
          {arriveTangent=0.7691, interpMode=RCIM_Cubic, leaveTangent=0.7691, tangentMode=RCTM_User, time=1, value=0}
         ]}
       }}))
    ScaleA=true
    ScaleRGB=true
   )
   &ScaleSpriteSize(Uniform Curve Sprite Scale=DataInterface() {propertyValues {Curve {keys [
        {arriveTangent=4, interpMode=RCIM_Cubic, leaveTangent=4, time=0, value=0}
        {arriveTangent=4, interpMode=RCIM_Cubic, leaveTangent=4, time=0.1, value=0.4}
        {arriveTangent=-0.6667, interpMode=RCIM_Cubic, leaveTangent=-0.6667, time=0.2, value=0.8}
        {arriveTangent=-1.3333, interpMode=RCIM_Cubic, leaveTangent=-1.3333, time=0.4, value=0.2}
        <at 0.5 0.4 RCIM_Cubic>
        {arriveTangent=2, interpMode=RCIM_Cubic, leaveTangent=2, time=0.6, value=0.2}
        {arriveTangent=-0.5, interpMode=RCIM_Cubic, leaveTangent=-0.5, time=0.8, value=1}
        {arriveTangent=-5, interpMode=RCIM_Cubic, leaveTangent=-5, time=1, value=0}
       ]}}})
   &CurlNoiseForce(Noise Frequency=100, Pan Noise=false) {enabled=false}
   &VectorNoiseForce(Force Amount=50, Module.Use Random Noise Time Interval=false)
   &Drag(Drag=2, Use Linear Drag=true, Use Rotational Drag=false)
   &SolveForcesAndVelocity(Clamp Velocity=false, Limit Acceleration=false)
  ]
  rendererClasses [&NiagaraSpriteRendererProperties]
  renderers [{rendererClass=&NiagaraSpriteRendererProperties}]
 }
}
summary {
 systemName=NS_Pickaxe_Shiitake_Shaolin_Idle
 emitters [
  {
   emitterName=Gem
   rendererClasses [
    &NiagaraSpriteRendererProperties
    &NiagaraMeshRendererProperties
    &NiagaraLightRendererProperties
   ]
  }
  {emitterName=Sparkles, rendererClasses [&NiagaraSpriteRendererProperties]}
  {bEnabled=false, emitterName=BitsWorld, rendererClasses [&NiagaraSpriteRendererProperties]}
 ]
 userVariables [
  User.bIsFrontEndPreview {type=NiagaraBool}
  User.New Skeletal Mesh {
   type=NiagaraDataInterfaceSkeletalMesh
   defaultValue {
    struct=NiagaraExt_VariableValue_DataInterface
    value {dataInterface=&self:NiagaraDataInterfaceSkeletalMesh_0}
   }
  }
 ]
}
